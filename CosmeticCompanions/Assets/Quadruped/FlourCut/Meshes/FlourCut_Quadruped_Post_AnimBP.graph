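# Post-process AnimBP for the FlourCut quadruped cosmetic companion that adjusts clothing sleeve and buckle bones.
# Weighted multi-bone constraints with inverted alpha scaling drive both upper sleeves and buckles, then mirrored additive bone-space offsets nudge the buckles.
# The result feeds a FlourCut PoseAsset blend before output.
# On initialization, isChonky is set from the main companion instance's Stage1 scale variant tag, with no visible read in the graph.
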AnimBlueprint /CosmeticCompanions/Assets/Quadruped/FlourCut/Meshes/FlourCut_Quadruped_Post_AnimBP
target_skeleton=&/FlourCut_Quadruped_Skeleton
graphs [
 AnimGraph {nodes [
   n1 = "Constraint - dyn_L_clothingSleeve_upper (PP)"!(ComponentPose=Local To Component(LocalPose=Input Pose()), Alpha=1, ConstraintWeights_0=0.9, ConstraintWeights_1:float=0.1) -> (Pose) {
    alphaScaleBiasClamp {bias=1, scale=-1}
    boneToModify {boneName=dyn_L_clothingSleeve_upper}
    constraintSetup [{targetBone {boneName=shoulder_fr_l}}, {targetBone {boneName=spine_03}}]
    constraintWeights [1,1]
   }
   n2 = "Constraint - dyn_R_clothingSleeve_upper (PP)"!(ComponentPose=@n1, Alpha=1, ConstraintWeights_0=0.9, ConstraintWeights_1:float=0.1) -> (Pose) {
    alphaScaleBiasClamp {bias=1, scale=-1}
    boneToModify {boneName=dyn_R_clothingSleeve_upper}
    constraintSetup [{targetBone {boneName=shoulder_fr_r}}, {targetBone {boneName=spine_03}}]
    constraintWeights [1,1]
   }
   n3 = "Constraint - L_buckle (PPPPPPPPP)"!(
    ComponentPose=@n2
    Alpha=1
    ConstraintWeights_0=0.0232
    ConstraintWeights_1:float=0.0035
    ConstraintWeights_2:float=0.8062
    ConstraintWeights_3:float=0.0002
    ConstraintWeights_4:float=0.0078
    ConstraintWeights_5:float=0.0013
    ConstraintWeights_6:float=0.0869
    ConstraintWeights_7:float=0.0005
    ConstraintWeights_8:float=0.0705
   ) -> (Pose) {
    alphaScaleBiasClamp {bias=0.75, scale=-0.75}
    boneToModify {boneName=L_buckle}
    constraintSetup [
     {targetBone {boneName=spine_03}}
     {targetBone {boneName=scapula_fr_l}}
     {targetBone {boneName=spine_02}}
     {targetBone {boneName=thigh_bk_l}}
     {targetBone {boneName=hips}}
     {targetBone {boneName=neck_01}}
     {targetBone {boneName=shoulder_fr_l}}
     {targetBone {boneName=neck_02}}
     {targetBone {boneName=spine_01}}
    ]
    constraintWeights [1,1,1,1,1,1,1,1,1]
   }
   n4 = "Constraint - R_buckle (PPPPPPPPP)"!(
    ComponentPose=@n3
    Alpha=1
    ConstraintWeights_0=0.0232
    ConstraintWeights_1:float=0.0035
    ConstraintWeights_2:float=0.8062
    ConstraintWeights_3:float=0.0002
    ConstraintWeights_4:float=0.0078
    ConstraintWeights_5:float=0.0013
    ConstraintWeights_6:float=0.0869
    ConstraintWeights_7:float=0.0005
    ConstraintWeights_8:float=0.0705
   ) -> (Pose) {
    alphaScaleBiasClamp {bias=0.75, scale=-0.75}
    boneToModify {boneName=R_buckle}
    constraintSetup [
     {targetBone {boneName=spine_03}}
     {targetBone {boneName=scapula_fr_r}}
     {targetBone {boneName=spine_02}}
     {targetBone {boneName=thigh_bk_r}}
     {targetBone {boneName=hips}}
     {targetBone {boneName=neck_01}}
     {targetBone {boneName=shoulder_fr_r}}
     {targetBone {boneName=neck_02}}
     {targetBone {boneName=spine_01}}
    ]
    constraintWeights [1,1,1,1,1,1,1,1,1]
   }
   n5 = "Transform (Modify) Bone - Bone: L_buckle"!(ComponentPose=@n4, Alpha=1, Translation="-0.5,0,0", Rotation="0,0,0") -> (Pose) {translationMode=BMM_Additive, translationSpace=BCS_BoneSpace, boneToModify {boneName=L_buckle}, translation {x=-0.5}}
   n6 = "Transform (Modify) Bone - Bone: R_buckle"!(ComponentPose=@n5, Alpha=1, Translation="0.5,0,0", Rotation="0,0,0") -> (Pose) {translationMode=BMM_Additive, translationSpace=BCS_BoneSpace, boneToModify {boneName=R_buckle}, translation {x=0.5}}
   Output Pose(Result=Pose!(
     PoseAsset:Pose Asset Object Reference=Get PoseAsset() -> (PoseAsset:Pose Asset Object Reference)
     SourcePose:Pose Link Structure=Component To Local(ComponentPose=@n6)
    ) -> (Pose:Pose Link Structure) {is=title AnimGraphNode_PoseBlendNode, blendOption=Linear})
  ]}
 EventGraph {nodes [
   Event BlueprintInitializeAnimation()
   n1 = GetMainAnimInstance!(self:Companion Post Process Anim Instance Object Reference) -> (ReturnValue:Fort Player Companion Anim Instance Object Reference)
   > Set isChonky(isChonky:double="To Float (Boolean)"(InBool=HasVariantTagForChannel!(
      self:Fort Player Companion Anim Instance Object Reference=@n1
      ChannelTag:Gameplay Tag Structure=(TagName="Cosmetics.Variant.Channel.Immutable.Scale")
      VariantTag:Gameplay Tag Structure=(TagName="Cosmetics.Variant.Property.Stage1")
     ) -> (ReturnValue:bool))) -> (Output_Get:double)
  ]}
]
variables [
 isChonky {type=number}
 PoseAsset {type=PoseAsset, default=&/FlourCut_Quadruped_PoseAsset}
]
